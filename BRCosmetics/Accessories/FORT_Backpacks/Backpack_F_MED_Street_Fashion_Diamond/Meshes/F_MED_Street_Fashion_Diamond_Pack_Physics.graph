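# Physics asset for the Street Fashion Diamond backpack cosmetic, defining collision bodies and joints for its dangling parts.
# The pack body combines boxes and capsules, while each dyn_ bone for pouches, straps, buckles and zipper has a single box.
# Pouches, main buckle and zipper constrain to the pack, with straps and side buckles chained to their pouch.
# Pouches and main buckle have tight angular limits; the zipper and straps swing much wider.
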PhysicsAsset /BRCosmetics/Accessories/FORT_Backpacks/Backpack_F_MED_Street_Fashion_Diamond/Meshes/F_MED_Street_Fashion_Diamond_Pack_Physics
bodies {
 dyn_l_buckle_bot [{shapeType=Box, extentX=0.8954, extentY=3.1621, extentZ=1.9874, center=<1.1572 0.0092 0>}]
 dyn_l_buckle_top [{shapeType=Box, extentX=1.1255, extentY=2.8875, extentZ=1.9213, center=<1.0753 0.1439 0>}]
 dyn_l_pouch [{shapeType=Box, extentX=4.7185, extentY=6.7966, extentZ=9.1822, center=<1.7275 -0.2941 -0.6416>}]
 dyn_l_pouch_strap [{shapeType=Box, extentX=0.8513, extentY=1.4777, extentZ=2.0125, center=<0.4372 -0.032 -0.8845>}]
 dyn_main_buckle [{shapeType=Box, extentX=3.1789, extentY=1.0215, extentZ=5.3763, center=<1.0368 -0.0383 -4.131>, rotation=<rot -2.6937 -90 -4.742>}]
 dyn_r_buckle_bot [{shapeType=Box, extentX=0.9098, extentY=2.7924, extentZ=1.99, center=<1.1094 -0.0171 0>}]
 dyn_r_buckle_top [{shapeType=Box, extentX=1.4329, extentY=1.9861, extentZ=4.3909, center=<0.9781 -0.0636 0.004>, rotation=<rot 0 -8.0963 -87.1288>}]
 dyn_r_pouch [{shapeType=Box, extentX=4.6778, extentY=6.654, extentZ=9.1702, center=<1.7349 0.2252 -0.6523>}]
 dyn_r_pouch_strap [{shapeType=Box, extentX=0.8111, extentY=1.4776, extentZ=1.8361, center=<0.419 0.0338 -0.978>}]
 dyn_zipper [{shapeType=Box, extentX=0.7324, extentY=2.2274, extentZ=2.2973, center=<0.2916 0.635 -2.2567>, rotation=<rot -15.783 -84.804 -0.6345>}]
 pack [
  {shapeType=Capsule, length=4.8522, radius=2.9019, center=<3.6999 -6.0851 -10.9219>, rotation=<rot 0 0 -15.0048>}
  {shapeType=Capsule, length=4.8514, radius=2.975, center=<-3.8759 -5.8139 -10.8166>, rotation=<rot 0 0 -15.0051>}
  {shapeType=Box, extentX=9.9551, extentY=12.6784, extentZ=10.2338, center=<0 -5.8957 1.5652>, rotation=<rot 0 0 -15.0152>}
  {shapeType=Box, extentX=6.1556, extentY=8.7634, extentZ=7.5, center=<-5.8624 -2.2981 1.823>, rotation=<rot 5 15.0166 0>}
  {shapeType=Box, extentX=7.7004, extentY=3.7767, extentZ=7.5, center=<-6.6984 0.6104 3.4085>, rotation=<rot 4.973 35.947 0.5181>}
 ]
}
constraints [
 {bone1Name=dyn_main_buckle, bone2Name=pack, swing1LimitDegrees=5, swing2LimitDegrees=7, twistLimitDegrees=3}
 {bone1Name=dyn_r_pouch, bone2Name=pack, swing1LimitDegrees=5, swing2LimitDegrees=3, twistLimitDegrees=3}
 {bone1Name=dyn_l_pouch, bone2Name=pack, swing1LimitDegrees=5, swing2LimitDegrees=3, twistLimitDegrees=3}
 {bone1Name=dyn_zipper, bone2Name=pack, swing1LimitDegrees=30, swing2LimitDegrees=45, twistLimitDegrees=120}
 {bone1Name=dyn_l_pouch_strap, bone2Name=dyn_l_pouch, swing1LimitDegrees=1, swing2LimitDegrees=75, twistLimitDegrees=3}
 {bone1Name=dyn_r_pouch_strap, bone2Name=dyn_r_pouch, swing1LimitDegrees=1, swing2LimitDegrees=75, twistLimitDegrees=3}
 {bone1Name=dyn_r_buckle_top, bone2Name=dyn_r_pouch, swing1LimitDegrees=5, swing2LimitDegrees=25, twistLimitDegrees=5}
 {bone1Name=dyn_r_buckle_bot, bone2Name=dyn_r_pouch, swing1LimitDegrees=10, swing2LimitDegrees=15, twistLimitDegrees=7}
 {bone1Name=dyn_l_buckle_bot, bone2Name=dyn_l_pouch, swing1LimitDegrees=5, swing2LimitDegrees=25, twistLimitDegrees=5}
 {bone1Name=dyn_l_buckle_top, bone2Name=dyn_l_pouch, swing1LimitDegrees=10, swing2LimitDegrees=15, twistLimitDegrees=7}
]
